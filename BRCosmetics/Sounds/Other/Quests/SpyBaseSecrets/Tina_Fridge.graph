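# Sound cue for the Tina room fridge sound in the Spy Base Secrets quest audio.
# A fixed delay node feeds a single wave player for the Tina_Room_Fridge wave, so playback starts about half a second after triggering.
# It uses the shared item-interact attenuation and the non-critical in-game UI sound class.
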SoundCue /BRCosmetics/Sounds/Other/Quests/SpyBaseSecrets/Tina_Fridge
root=Delay_0
nodes [
 {id=Delay_0, children [WavePlayer_0], properties {delayMax=0.55, delayMin=0.55}}
 {id=WavePlayer_0, properties {soundWaveAssetPtr=&/Tina_Room_Fridge}}
]
properties {
 attenuationSettings=/Game/Sounds/Attenuation/Fort_Item_Interact_Att
 maxDistance=1850
 soundClassObject=&Fort_UI_InGame_NonCritical
 volumeMultiplier=1
}
