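# Sound cue for the Tina machine blast in the Spy Base Secrets quest sounds.
# A single wave player feeds the Tina_Machine_Explo wave straight to the output, with no randomization, modulation or mixing nodes.
# It uses the non-critical in-game UI sound class with a boosted volume, and sets no attenuation.
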SoundCue /BRCosmetics/Sounds/Other/Quests/SpyBaseSecrets/Tina_Machine_Blast_Cue
root=WavePlayer_0
nodes [{id=WavePlayer_0, properties {soundWaveAssetPtr=&/Tina_Machine_Explo}}]
properties {maxDistance=2097152, soundClassObject=&Fort_UI_InGame_NonCritical, volumeMultiplier=1.25}
