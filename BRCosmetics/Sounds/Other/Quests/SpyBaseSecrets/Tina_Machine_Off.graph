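# Sound cue for the Tina machine powering off, part of the Spy Base Secrets quest sounds.
# It plays a single stop wave through one wave player, with no randomization or layering.
# It uses item interaction attenuation for short-range spatial playback, reduced volume and the non-critical in-game UI sound class.
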SoundCue /BRCosmetics/Sounds/Other/Quests/SpyBaseSecrets/Tina_Machine_Off
root=WavePlayer_2
nodes [{id=WavePlayer_2, properties {soundWaveAssetPtr=&/Tina_Machine_Stop}}]
properties {
 attenuationSettings=/Game/Sounds/Attenuation/Fort_Item_Interact_Att
 maxDistance=1850
 soundClassObject=&Fort_UI_InGame_NonCritical
 volumeMultiplier=0.55
}
