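# Sound cue for stage 3 of the Tina machine, under the Spy Base Secrets quest sounds, using item-interact attenuation.
# A Mixer layers a one-shot machine start wave with a looping stage 3 wave at slightly lower level.
# An Enveloper ramps the loop's volume up from quiet over its first moments, under the start sound.
# Concurrency allows one instance per owner, stopping the oldest with a release.
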SoundCue /BRCosmetics/Sounds/Other/Quests/SpyBaseSecrets/Tina_Machine_Stage3_Cue
root=Mixer_1
nodes [
 {
  id=Enveloper_1
  children [WavePlayer_0]
  properties {pitchMax=2, volumeMax=4, pitchCurve {editorCurveData {keys [{time=0}]}}, volumeCurve {editorCurveData {keys [<at 0 0.1>,<at 0.4 1>]}}}
 }
 {id=Mixer_1, children [WavePlayer_1,Enveloper_1], properties {inputVolume [1,0.8]}}
 {id=WavePlayer_0, properties {bLooping=true, soundWaveAssetPtr=&/Tina_Machine_Stage3}}
 {id=WavePlayer_1, properties {soundWaveAssetPtr=&/Tina_Machine_Start}}
]
properties {
 attenuationSettings=/Game/Sounds/Attenuation/Fort_Item_Interact_Att
 maxDistance=1850
 soundClassObject=&Fort_UI_InGame_NonCritical
 volumeMultiplier=0.65
 concurrencySet [/Game/Sounds/Concurrency_Rules/OneMaximum_Owner_StopOldestRelease]
}
